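# Physics asset for the Informer Male one-handed pickaxe cosmetic, giving its mesh a handle body and a dangling four-segment strap.
# A capsule covers the handle bone, and small boxes cover the chained dyn_strap bones.
# Angular constraints link each strap segment to the previous one, with the first anchored to the handle.
# The root joint has a tighter swing limit, and every joint allows very little twist.
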PhysicsAsset /BRCosmetics/Weapons/FORT_Melee/Pickaxe_Informer_Male_1h/Meshes/Informer_Male_1h_Physics
bodies {
 Melee_TwoHanded_Handle [{shapeType=Capsule, length=49.9442, radius=2.4199, center=<0 -8.2916 42.8319>}]
 dyn_strap_01 [{shapeType=Box, extentX=3, extentY=3, extentZ=2, center=<1 0 0>}]
 dyn_strap_02 [{shapeType=Box, extentX=4, extentY=3, extentZ=2, center=<2 0 0>}]
 dyn_strap_03 [{shapeType=Box, extentX=4, extentY=3, extentZ=2, center=<2 0 0>}]
 dyn_strap_04 [{shapeType=Box, extentX=6, extentY=3, extentZ=2, center=<3 0 0>}]
}
constraints [
 {bone1Name=dyn_strap_01, bone2Name=Melee_TwoHanded_Handle, swing1LimitDegrees=15, swing2LimitDegrees=60, twistLimitDegrees=5}
 {bone1Name=dyn_strap_02, bone2Name=dyn_strap_01, swing1LimitDegrees=45, swing2LimitDegrees=60, twistLimitDegrees=5}
 {bone1Name=dyn_strap_03, bone2Name=dyn_strap_02, swing1LimitDegrees=45, swing2LimitDegrees=60, twistLimitDegrees=5}
 {bone1Name=dyn_strap_04, bone2Name=dyn_strap_03, swing1LimitDegrees=45, swing2LimitDegrees=60, twistLimitDegrees=5}
]
